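; Outdoor ShadyBill sign prop that shows a language-specific message on its LED billboard material.
; On BeginPlay, where cosmetic events can run, it looks up the current language in LocImageTable and async-loads the mapped texture.
; It feeds that texture into a dynamic material instance, with a scalable-float value driving LB_ImageSequence_Speed to select the message variant.
; Failed lookups and casts print debug strings, and the cleanup function is empty.
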(event EventBeginPlay
  (|CallParentFunction)
  (Networking|CanExecuteCosmeticEvents
    (:True
      (CallFunction|EventSetupMessagePanelforLocalization))))

(event Collision|EventActorBeginOverlap
  (params (OtherActor :type Actor%20Object%20Reference))
  (|CallParentFunction OtherActor))

(event EventTick
  (params (DeltaSeconds :type Float%20%28single-precision%29))
  (|CallParentFunction DeltaSeconds) ; This node is disabled and will not be called.\nDrag off pins to build functionality.
)

(event Custom|EventSetupMessagePanelForLocalization
  (bind _returnvalue (Utilities|Internationalization|GetCurrentLanguage))
  (bind _locimagetable (Variables|Default|GetLocImageTable))
  (bind (_value _returnvalue_2) (Utilities|Map|Find _locimagetable _returnvalue))
  (bind _textouse_2_b69e6377402e9a2d409c2ab94fc6cff1 (Utilities|Struct|BreakWildEstateLEDMessageInfo _value))
  (if _returnvalue_2
    (if (not (Asset|IsAssetNull _textouse_2_b69e6377402e9a2d409c2ab94fc6cff1))
      (bind _object (Utilities|AsyncLoadAsset _textouse_2_b69e6377402e9a2d409c2ab94fc6cff1)
        (:Completed
          (bind _astexture_2d (Utilities|Casting|CastToTexture2D _object)
            (:then
              (bind _returnvalue_3 (Rendering|Material|CreateDynamicMaterialInstance (Variables|Debug|GetStaticMeshComponent) 0 (Rendering|Material|GetMaterial (Variables|Debug|GetStaticMeshComponent) 0) "" :!class /Script/Engine.PrimitiveComponent))
              (Variables|Default|SetDynamicMaterial _returnvalue_3)
              (Rendering|Material|SetTextureParameterValue _returnvalue_3 "LB_LivingBillboardMessage" _astexture_2d) ; Specify texture to load / use
              (Rendering|Material|SetScalarParameterValue _returnvalue_3 "LB_ImageSequence_Speed" (ScalableFloat|GetValueatLevel (Variables|Default|GetHFRow_TextVariantDisplayValue)) :!class /Script/Engine.MaterialInstanceDynamic) ; Specify message
            )
            (:CastFailed
              (Development|PrintString "Failed to case Text asset to Texture!")))))
      (else
        (Development|PrintString "Text asset not found for ShadyBill_OutdoorSIgn")))
    (else
      (Development|PrintString (Utilities|String|Append (Utilities|String|Append "Current language string (" _returnvalue) ") not found in LocImageTable!")))))

(fn EventCleanupMessagePanelForLocalization)

(fn ConstructionScript
  (|CallParentFunction))
